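# Translucent, unlit, two-sided smoke material for the Vivid Trooper cosmetic's feet, usable on Niagara mesh particles.
# A UV-driven gradient blends purple to green, scaled by time of day for emissive color.
# Panning cloud noise plus a second gradient forms an eroded mask, shaped by a sphere mask, particle alpha and depth fade.
# Time-scrolled curl noise, weighted by vertex red, creates animated displacement, replaced by a constant on mobile.
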Material /BRCosmetics/Effects/Fort_Effects/Effects/Characters/Athena_Parts/Vivid/M_VividTrooper_SmokeFeet
graph {
 properties {
  bCastRayTracedShadows=false
  bUsedWithInstancedStaticMeshes=true
  bUsedWithNiagaraMeshParticles=true
  blendMode=BLEND_Translucent
  pixelDepthOffsetMode=PDOM_Legacy
  shadingModel=MSM_Unlit
  twoSided=true
  usedShadingModels=Unlit
 }
 nodes [
  n1 = TextureCoordinate() {coordinateIndex=1}
  ComponentMask(None=@n1) {r=true}
  n2 = Constant3Vector() {constant=<rgba 0.1095 0.016 0.8963 1>}
  n3 = Constant3Vector() {constant=<rgba 0.1119 0.8148 0.3813 1>}
  n4 = MaterialFunctionCall(Float2=@n1) {materialFunction=&BreakOutFloat2Components}
  n5 = OneMinus(None=@n4.R)
  n6 = Multiply(A=TextureSample(UVs=MaterialFunctionCall(X=@n5, Y=@n4.G) {materialFunction=&MakeFloat2}.Result) {texture=/Game/Missions/Primary/OB_FarmsteadFort/ArtAssets/Textures/T_ProjectorGradient}.R) {constB=0.8}
  n7 = Constant() {r=0.3}
  n8 = Constant() {r=1}
  n9 = Constant() {r=0.5143}
  MaterialFunctionCall(Emissive=Multiply(
    A=LinearInterpolate(A=@n2, B=@n3, Alpha=@n6)
    B=MaterialFunctionCall(Value_Morning=@n7, Value_Day=@n8, Value_Evening=@n7, Value_Night=@n9) {materialFunction=&Scalar_Time_of_Day}.Result
   )) {materialFunction=&MF_FX_EmissiveAdjustments}
  n10 = TextureSample(UVs=Panner(Coordinate=TextureCoordinate() {uTiling=0.2, vTiling=0.45}) {speedX=-0.2, speedY=0.05}) {samplerType=SAMPLERTYPE_LinearColor, texture=/Game/Global/Textures/Default/Tileable_Patterns/Cloud_Diffuse}
  n11 = MaterialFunctionCall(Float2=TextureCoordinate() {coordinateIndex=1}) {materialFunction=&BreakOutFloat2Components}
  n12 = TextureSample(UVs=MaterialFunctionCall(X=Multiply(A=@n11.R) {constB=-1}, Y=@n11.G) {materialFunction=&MakeFloat2}.Result) {texture=/BRCosmetics/Characters/Player/Male/Medium/Bodies/M_MED_Rust_Remix/Textures/Gradients/T_Rust_Remix_T1+T2_v1_Gradient_Armor.T_Rust_Remix_T1+T2_v1_Gradient_Armor}
  n13 = Add(A=Power(Base=OneMinus(None=@n10.R)) {constExponent=2.5}, B=@n12.R)
  n14 = Constant2Vector()
  n15 = ScalarParameter() {defaultValue=0.5, parameterName=Radius}
  n16 = ScalarParameter() {defaultValue=58.4769, parameterName=hardness}
  n17 = OneMinus(None=Clamp(None=SphereMask(A=@n13, B=@n14, Radius=@n15, Hardness=@n16)))
  n18 = SphereMask(A=TextureCoordinate() {coordinateIndex=1, uTiling=0.5}, B=Constant2Vector() {g=0.5}) {attenuationRadius=0.5, hardnessPercent=0}
  Multiply(A=Multiply(A=Multiply(A=@n17, B=@n18) {constB=-1}, B=ParticleColor().A), B=DepthFade() {fadeDistanceDefault=5})
  n19 = ScalarParameter() {defaultValue=4, parameterName=Curl_Intensity}
  n20 = Subtract(A=WorldPosition().XYZ, B=ActorPositionWS())
  n21 = Multiply(A=ScalarParameter() {defaultValue=5, parameterName=Curl_Speed}, B=Time())
  n22 = VectorNoise(World Position=Divide(A=Subtract(A=@n20, B=@n21), B=ScalarParameter() {defaultValue=10, parameterName=Curl_Scale})) {noiseFunction=VNF_VectorALU}
  ShadingPathSwitch(Default=Multiply(A=VertexColor().R, B=Multiply(A=@n19, B=@n22)), Mobile=Constant() {r=1}) {desc=Migrated from FeatureLevelSwitch}
 ]
}
