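# Translucent, unlit, two-sided material for the Vivid Trooper's animated purple smoke head, usable on Niagara mesh particles.
# Panning turbulence, noise and mask textures blend color parameters into the emissive, with a dynamic-parameter time offset.
# Scrolling noise, gradient and ramp textures on U-flipped UV1 form a depth-faded mask.
# Switchable world position offset adds sine waves and oscillating twist, weighted by vertex red and a powered bounding-box height gradient.
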Material /BRCosmetics/Effects/Fort_Effects/Effects/Characters/Athena_Parts/Vivid/M_VividTrooper_Smokehead
graph {
 properties {
  bCastRayTracedShadows=false
  bUsedWithInstancedStaticMeshes=true
  bUsedWithNiagaraMeshParticles=true
  blendMode=BLEND_Translucent
  pixelDepthOffsetMode=PDOM_Legacy
  shadingModel=MSM_Unlit
  twoSided=true
  usedShadingModels=Unlit
 }
 nodes [
  n1 = Time()
  n2 = DynamicParameter() {paramNames [Time offset, Param2, PowerMultiplier, Param4]}
  Add(A=@n1, B=@n2.Time offset)
  n3 = Power(Base=DotProduct(A=CameraVectorWS(), B=VertexNormalWS()), Exp=Constant() {r=1})
  Clamp(None=Multiply(B=Clamp(None=@n3)))
  n4 = VectorParameter() {parameterName=Color_Detail, defaultValue=<rgba 0.3698 0.0562 0.516 1>}
  n5 = Add(A=DynamicParameter() {paramNames [Time offset, Param2, PowerMultiplier, Param4]}.Time offset, B=Time())
  n6 = TextureSample(UVs=Panner(Coordinate=TextureCoordinate() {uTiling=0.581}, Time=@n5) {speedX=-0.8, speedY=-0.08}) {texture=/Game/Global/Textures/Default/Caustics/T_Turbulence_Seamless}
  n7 = TextureCoordinate()
  n8 = LinearInterpolate(
   A=VectorParameter() {parameterName=BaseColor_Bright, defaultValue=<rgba 0.1226 0.0274 0.2865 1>}.RGB
   B=VectorParameter() {parameterName=BaseColor_Dark, defaultValue=<rgba 0.0176 0.0018 0.1384 1>}.RGB
   Alpha=TextureSample(UVs=Panner(Coordinate=@n7, Time=@n5) {speedX=-0.3, speedY=-0.01}) {samplerType=SAMPLERTYPE_LinearColor, texture=/Game/Effects/Fort_Effects/BluGlow/Violentium/Violentium_Mask}.G
  )
  n9 = TextureSample(UVs=Panner(Coordinate=TextureCoordinate(), Time=@n5) {speedX=-1, speedY=0.1}) {texture=/Game/Global/Textures/Default/Noise/T_Tiling_Noise_01}
  n10 = LinearInterpolate(A=@n9.RGB) {constA=1}
  MaterialFunctionCall(Emissive=Add(A=LinearInterpolate(B=@n4.RGB, Alpha=@n6.G), B=Multiply(A=@n8, B=@n10))) {materialFunction=&MF_FX_EmissiveAdjustments}
  n11 = TextureSample(UVs=Panner(Coordinate=TextureCoordinate() {uTiling=0.5714}, Time=@n5) {speedX=-1, speedY=0.1}) {texture=&T_Tiling_Noise_01}
  n12 = MaterialFunctionCall(Float2=TextureCoordinate() {coordinateIndex=1}) {materialFunction=&BreakOutFloat2Components}
  n13 = MaterialFunctionCall(X=Multiply(A=@n12.R) {constB=-1}, Y=@n12.G) {materialFunction=&MakeFloat2}
  n14 = TextureSample(UVs=@n13.Result) {texture=/BRCosmetics/Characters/Player/Male/Medium/Bodies/M_MED_Rust_Remix/Textures/Gradients/T_Rust_Remix_T1+T2_v1_Gradient_Armor.T_Rust_Remix_T1+T2_v1_Gradient_Armor}
  n15 = Multiply(A=Multiply(A=Power(Base=@n11.RGB) {constExponent=0.7}) {constB=3}, B=Power(Base=Multiply(A=@n14.RGB) {constB=3}) {constExponent=1.5}) {constB=2}
  n16 = TextureSample(UVs=@n13.Result) {texture=/Game/Effects/Fort_Effects/Textures/Ramps/T_Ribbon_Ramp}
  Multiply(A=Multiply(A=Clamp(None=@n15) {minDefault=-1}, B=Multiply(A=@n16.G) {constB=2}) {constB=2}, B=DepthFade() {fadeDistanceDefault=5})
  Multiply(A=Power(Base=@n9.RGB, Exp=ParticleColor().A))
  Multiply(A=ComponentMask(None=ObjectPositionWS()) {r=true}, B=ScalarParameter() {parameterName=PositionInfluence})
  ParticlePositionWS()
  n17 = ScalarParameter() {defaultValue=1.2465, parameterName=Power}
  n18 = DynamicParameter() {paramNames [Time offset, Param2, PowerMultiplier, Param4]}
  n19 = Add(
   A=Power(
    Base=MaterialFunctionCall() {materialFunction=&BoundingBoxBased_0-1_UVW}.B
    Exp=Multiply(A=@n17, B=@n18.PowerMultiplier)
   )
   B=ScalarParameter() {defaultValue=-1.0492, parameterName=Offset}
  )
  n20 = TransformPosition(None=WorldPosition().XYZ) {transformSourceType=TRANSFORMPOSSOURCE_World}
  n21 = Add(
   A=Multiply(A=ComponentMask(None=@n20) {b=true}, B=ScalarParameter() {defaultValue=0.01, parameterName=Freq})
   B=Multiply(A=Multiply(A=@n1, B=@n2.Time offset), B=ScalarParameter() {defaultValue=-1.5, parameterName=Speed})
  )
  n22 = Transform(None=VectorParameter() {parameterName=HeightOffset_Top, defaultValue=<rgba 15 0 0 0>}.RGB) {transformSourceType=TRANSFORMSOURCE_Local}
  n23 = Transform(None=VectorParameter() {parameterName=HeightOffset_Front, defaultValue=<rgba 15 0 0 0>}.RGB) {transformSourceType=TRANSFORMSOURCE_Local}
  n24 = Normalize(VectorInput=TransformPosition(None=ObjectPositionWS()) {transformType=TRANSFORMPOSSOURCE_Camera})
  n25 = Add(A=Abs(None=ComponentMask(None=@n24) {r=true}), B=Abs(None=ComponentMask(None=@n24) {g=true}))
  n26 = ScalarParameter() {defaultValue=-4, parameterName=TwistSpeed}
  n27 = Sine(None=Add(A=Multiply(A=@n1, B=@n26), B=@n19)) {period=2}
  n28 = ScalarParameter() {defaultValue=-0.2, parameterName=Twist}
  n29 = Add(
   A=Multiply(A=Sine(None=@n21), B=LinearInterpolate(A=@n22, B=@n23, Alpha=@n25))
   B=RotateAboutAxis(
    NormalizedRotationAxis=Constant3Vector() {constant=<rgba 1 1 1 0>}
    RotationAngle=Multiply(A=@n27, B=@n28)
    PivotPoint=ObjectPositionWS()
    Position=WorldPosition().XYZ
   )
  )
  StaticSwitchParameter(True=Multiply(A=VertexColor().R, B=Multiply(A=@n19, B=@n29)), False=Constant()) {defaultValue=true, parameterName=USE_WPO}
  TextureCoordinate()
 ]
}
