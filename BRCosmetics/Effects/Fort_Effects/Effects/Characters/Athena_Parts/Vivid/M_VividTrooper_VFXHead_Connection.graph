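# Translucent, unlit, two-sided effect material for the Vivid Trooper head connection, drawing flowing purple energy.
# Panning mask, noise and turbulence textures blend tinted colors into an emissive scaled by time of day.
# Panning noise shaped by UV ramps, with optional horizontal falloff, multiplies a smoke-broken camera-facing term to form an opacity mask.
# Switchable vertex offset uses scrolling actor-relative vector noise masked by vertex red, replaced by a constant on mobile.
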Material /BRCosmetics/Effects/Fort_Effects/Effects/Characters/Athena_Parts/Vivid/M_VividTrooper_VFXHead_Connection
graph {
 properties {
  bCastRayTracedShadows=false
  bUsedWithGeometryCollections=true
  bUsedWithMeshDeformer=true
  bUsedWithNiagaraMeshParticles=true
  bUsedWithNiagaraSprites=true
  bUsedWithSkeletalMesh=true
  bUsedWithSplineMeshes=true
  blendMode=BLEND_Translucent
  ditherOpacityMask=true
  pixelDepthOffsetMode=PDOM_Legacy
  shadingModel=MSM_Unlit
  twoSided=true
  usedShadingModels=Unlit
 }
 nodes [
  n1 = DotProduct(A=CameraVectorWS(), B=VertexNormalWS())
  n2 = Multiply(A=@n1)
  n3 = Constant() {r=0.5}
  n4 = Panner(Coordinate=TextureCoordinate() {uTiling=2}) {bFractionalPart=true, speedX=0.3, speedY=1}
  n5 = Multiply(A=Power(Base=TextureSample(UVs=@n4) {samplerType=SAMPLERTYPE_LinearColor, texture=/Game/Global/Textures/Default/Tileable_Patterns/T_Tile_Smoke_Small}.R) {constExponent=5}) {constB=5}
  n6 = Clamp(None=Multiply(A=@n1, B=Clamp(None=@n5) {minDefault=-1}))
  n7 = Add(A=Power(Base=@n1) {constExponent=0.5}, B=@n6)
  Add(A=Clamp(None=Power(Base=@n2, Exp=@n3)), B=@n7) {constB=0}
  Clamp(None=Multiply(B=@n6))
  n8 = Constant() {r=0.3}
  n9 = Constant() {r=1}
  n10 = Constant() {r=0.5143}
  n11 = MaterialFunctionCall(Float2=TextureCoordinate() {uTiling=2, vTiling=0.3}) {materialFunction=&BreakOutFloat2Components}
  n12 = MaterialFunctionCall(X=Multiply(A=@n11.R, B=ScalarParameter() {defaultValue=1, parameterName=UV-Scale_Detail}), Y=@n11.G) {materialFunction=&MakeFloat2}
  n13 = Add(
   A=DynamicParameter() {paramNames [Time offset, Param2, PowerMultiplier, Param4]}.Time offset
   B=Add(A=Time(), B=ScalarParameter() {parameterName=TextureTimeOffset})
  )
  n14 = Multiply(A=TextureSample(UVs=Panner(Coordinate=@n12.Result, Time=@n13) {speedX=-0.08, speedY=0.8}) {texture=/Game/Global/Textures/Default/Caustics/T_Turbulence_Seamless}.G) {constB=2}
  n15 = LinearInterpolate(B=VectorParameter() {parameterName=Color_Detail, defaultValue=<rgba 0.1344 0.0204 0.1875 1>}.RGB, Alpha=Clamp(None=@n14))
  n16 = MaterialFunctionCall(Float2=TextureCoordinate() {uTiling=2, vTiling=0.3}) {materialFunction=&BreakOutFloat2Components}
  n17 = MaterialFunctionCall(X=Multiply(A=@n16.R, B=ScalarParameter() {defaultValue=1, parameterName=UV-Uscale}), Y=@n16.G) {materialFunction=&MakeFloat2}
  n18 = Panner(Coordinate=@n17.Result, Time=@n13) {speedX=-0.01, speedY=0.3}
  n19 = Panner(Coordinate=@n17.Result, Time=@n13) {speedX=-0.2, speedY=1}
  n20 = Multiply(
   A=LinearInterpolate(
    A=VectorParameter() {parameterName=BaseColor_Bright, defaultValue=<rgba 0.1226 0.0274 0.2865 1>}.RGB
    B=VectorParameter() {parameterName=BaseColor_Dark, defaultValue=<rgba 0.0176 0.0018 0.1384 1>}.RGB
    Alpha=TextureSample(UVs=@n18) {samplerType=SAMPLERTYPE_LinearColor, texture=/Game/Effects/Fort_Effects/BluGlow/Violentium/Violentium_Mask}.G
   )
   B=LinearInterpolate(A=TextureSample(UVs=@n19) {texture=/Game/Global/Textures/Default/Noise/T_Tiling_Noise_01}.RGB) {constA=1}
  )
  MaterialFunctionCall(Emissive=Multiply(
    A=MaterialFunctionCall(Value_Morning=@n8, Value_Day=@n9, Value_Evening=@n8, Value_Night=@n10) {materialFunction=&Scalar_Time_of_Day}.Result
    B=Add(A=@n15, B=@n20)
   )) {materialFunction=&MF_FX_EmissiveAdjustments}
  n21 = TextureSample(UVs=Panner(Coordinate=@n17.Result, Time=@n13) {speedX=-0.1, speedY=1}) {texture=&T_Tiling_Noise_01}
  n22 = TextureCoordinate()
  n23 = MaterialFunctionCall(Float2=@n22) {materialFunction=&BreakOutFloat2Components}
  n24 = Power(
   Base=TextureSample(UVs=MaterialFunctionCall(X=@n23.G, Y=@n23.G) {materialFunction=&MakeFloat2}.Result) {samplerType=SAMPLERTYPE_Masks, texture=/BRCosmetics/Weapons/WeaponSkins/Wraps/Textures/CircutLines/T_Wrap_CircutLines_Emissive_Ramp}.RGB
   Exp=ScalarParameter() {defaultValue=1.3, parameterName=vertical ramp crunch}
  ) {constExponent=1.3}
  n25 = TextureSample(UVs=@n22) {samplerType=SAMPLERTYPE_Masks, texture=&T_Wrap_CircutLines_Emissive_Ramp}
  n26 = ScalarParameter() {defaultValue=3, parameterName=horizontal ramp crunch}
  n27 = StaticSwitchParameter(True=Multiply(A=@n24, B=Power(Base=@n25.RGB, Exp=@n26) {constExponent=1.3}), False=@n24) {parameterName=USE_horizontalramp}
  n28 = Multiply(A=Multiply(A=Power(Base=@n21.RGB) {constExponent=0.9}) {constB=3}, B=@n27) {constB=2}
  Multiply(A=Clamp(None=Add(A=@n28, B=@n27)), B=@n7)
  Power(Base=@n14)
  n29 = Time()
  n30 = ScalarParameter() {parameterName=WPOTimeOffset}
  n31 = Subtract(
   A=Subtract(A=WorldPosition().XYZ, B=ActorPositionWS())
   B=Multiply(A=ScalarParameter() {defaultValue=50, parameterName=Curl_Speed}, B=Add(A=@n29, B=@n30))
  )
  n32 = ScalarParameter() {defaultValue=58.7749, parameterName=Curl_Scale}
  n33 = Multiply(
   A=ScalarParameter() {defaultValue=2, parameterName=Curl_Intensity}
   B=VectorNoise(World Position=Divide(A=@n31, B=@n32)) {noiseFunction=VNF_VectorALU}
  )
  n34 = Constant() {r=1}
  StaticSwitchParameter(True=Multiply(A=VertexColor().R, B=ShadingPathSwitch(Default=@n33, Mobile=@n34) {desc=Migrated from FeatureLevelSwitch}), False=Constant()) {defaultValue=true, parameterName=USE_WPO}
 ]
}
