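# Opaque, unlit effect material for the Vivid Trooper cosmetic's head, with animated energy, flame and glowing eyes.
# Panned noise textures, offset by a dynamic-parameter time, blend base colors with an optional gradient-driven dark area.
# Head-mask channels add noise-distorted side flame and switchable UV1 eyes, then time-of-day scaling feeds emissive adjustments.
# Separate branches compute view-angle and camera-fade masks, plus switchable vertex-color-masked curl-noise for WPO.
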Material /BRCosmetics/Effects/Fort_Effects/Effects/Characters/Athena_Parts/Vivid/M_VividTrooper_VFXHead_Opaque
graph {
 properties {
  bCastRayTracedShadows=false
  bUsedWithGeometryCollections=true
  bUsedWithMeshDeformer=true
  bUsedWithNiagaraMeshParticles=true
  bUsedWithNiagaraSprites=true
  bUsedWithSkeletalMesh=true
  bUsedWithSplineMeshes=true
  pixelDepthOffsetMode=PDOM_Legacy
  shadingModel=MSM_Unlit
  usedShadingModels=Unlit
 }
 nodes [
  n1 = DotProduct(A=CameraVectorWS(), B=VertexNormalWS())
  n2 = Panner(Coordinate=TextureCoordinate() {uTiling=2}) {bFractionalPart=true, speedX=0.3, speedY=1}
  n3 = Multiply(A=Power(Base=TextureSample(UVs=@n2) {samplerType=SAMPLERTYPE_LinearColor, texture=/Game/Global/Textures/Default/Tileable_Patterns/T_Tile_Smoke_Small}.R) {constExponent=5}) {constB=5}
  n4 = Clamp(None=Multiply(A=@n1, B=Clamp(None=@n3) {minDefault=-1}))
  Clamp(None=Multiply(B=@n4))
  n5 = Constant() {r=0.3}
  n6 = Constant() {r=1}
  n7 = Constant() {r=0.5143}
  n8 = TextureCoordinate()
  n9 = Add(A=DynamicParameter() {paramNames [Time offset, Param2, PowerMultiplier, Param4]}.Time offset, B=Time())
  n10 = Multiply(A=TextureSample(UVs=Panner(Coordinate=@n8, Time=@n9) {speedX=-0.08, speedY=0.8}) {texture=/Game/Global/Textures/Default/Caustics/T_Turbulence_Seamless}.G) {constB=2}
  n11 = VectorParameter() {parameterName=BaseColor_Bright, defaultValue=<rgba 0.1226 0.0274 0.2865 1>}
  n12 = VectorParameter() {parameterName=BaseColor_Dark, defaultValue=<rgba 0.0176 0.0018 0.1384 1>}
  n13 = TextureSample(UVs=Panner(Coordinate=TextureCoordinate(), Time=@n9) {speedX=-0.01, speedY=0.3}) {samplerType=SAMPLERTYPE_LinearColor, texture=/Game/Effects/Fort_Effects/BluGlow/Violentium/Violentium_Mask}
  n14 = TextureSample(UVs=Panner(Coordinate=TextureCoordinate(), Time=@n9) {speedX=-0.2, speedY=1}) {texture=/Game/Global/Textures/Default/Noise/T_Tiling_Noise_01}
  n15 = Add(
   A=LinearInterpolate(B=VectorParameter() {parameterName=Color_Detail, defaultValue=<rgba 0.1344 0.0204 0.1875 1>}.RGB, Alpha=Clamp(None=@n10))
   B=Multiply(A=LinearInterpolate(A=@n11.RGB, B=@n12.RGB, Alpha=@n13.G), B=LinearInterpolate(A=@n14.RGB) {constA=1})
  )
  n16 = VectorParameter() {parameterName=DarkAreaColor, defaultValue=<rgba 0.0104 0.0035 0.0084 1>}
  n17 = MaterialFunctionCall(Float2=TextureCoordinate()) {materialFunction=&BreakOutFloat2Components}
  n18 = TextureSample(UVs=MaterialFunctionCall(X=@n17.G, Y=Multiply(A=@n17.R) {constB=-1}) {materialFunction=&MakeFloat2}.Result) {texture=/BRCosmetics/Characters/Player/Male/Medium/Bodies/M_MED_Rust_Remix/Textures/Gradients/T_Rust_Remix_T1+T2_v1_Gradient_Armor.T_Rust_Remix_T1+T2_v1_Gradient_Armor}
  n19 = Clamp(None=Multiply(A=Power(Base=@n18.RGB) {constExponent=2.3}) {constB=10}) {maxDefault=0.85}
  n20 = Panner() {speedY=0.45}
  n21 = Subtract(A=Multiply(A=TextureSample(UVs=@n20) {samplerType=SAMPLERTYPE_Masks, texture=/Game/Gadgets/Assets/VinderTech_GliderChute/Glider_Vampire/Textures/T_Glider_Bat_Noises}.RGB) {constB=2})
  n22 = Add(A=TextureCoordinate(), B=Multiply(A=ComponentMask(None=@n21) {g=true, r=true}) {constB=0.05})
  n23 = LinearInterpolate(
   A=StaticSwitchParameter(True=LinearInterpolate(A=@n15, B=@n16.RGB, Alpha=@n19), False=@n15) {defaultValue=true, parameterName=USE_DarkArea}
   B=VectorParameter() {parameterName=SideFlameColor, defaultValue=<rgba 1.2 0.4161 0 1>}.RGB
   Alpha=Multiply(A=TextureSample(UVs=@n22) {texture=/BRCosmetics/Characters/Player/Male/Medium/Bodies/M_MED_Vivid_Trooper/Textures/T_M_MED_Vivid_Trooper_Head_Mask}.B) {constB=0.5}
  )
  n24 = VectorParameter() {parameterName=EyesGlowColor, defaultValue=<rgba 0.6719 0 0.5087 1>}
  n25 = TextureCoordinate() {coordinateIndex=1}
  n26 = TextureSample(UVs=Panner() {speedY=0.3}) {samplerType=SAMPLERTYPE_Masks, texture=&T_Glider_Bat_Noises}
  n27 = ComponentMask(None=Subtract(A=Multiply(A=@n26.RGB) {constB=2})) {g=true, r=true}
  n28 = TextureSample(UVs=Add(A=@n25, B=Multiply(A=@n27) {constB=0.06})) {texture=&T_M_MED_Vivid_Trooper_Head_Mask}
  n29 = Panner() {speedY=0.1}
  n30 = Subtract(A=Multiply(A=TextureSample(UVs=@n29) {samplerType=SAMPLERTYPE_Masks, texture=&T_Glider_Bat_Noises}.RGB) {constB=2})
  n31 = Add(A=@n25, B=Multiply(A=ComponentMask(None=@n30) {g=true, r=true}) {constB=0.02})
  n32 = VectorParameter() {parameterName=EyesColor, defaultValue=<rgba 0.9647 0 0.7379 1>}
  n33 = LinearInterpolate(
   A=LinearInterpolate(
    A=LinearInterpolate(A=@n23, B=@n24.RGB, Alpha=@n28.G)
    B=@n24.RGB
    Alpha=TextureSample(UVs=@n31) {texture=&T_M_MED_Vivid_Trooper_Head_Mask}.G
   )
   B=@n32.RGB
   Alpha=TextureSample(UVs=@n25) {texture=&T_M_MED_Vivid_Trooper_Head_Mask}.R
  )
  MaterialFunctionCall(Emissive=Multiply(
    A=MaterialFunctionCall(Value_Morning=@n5, Value_Day=@n6, Value_Evening=@n5, Value_Night=@n7) {materialFunction=&Scalar_Time_of_Day}.Result
    B=StaticSwitchParameter(True=@n33, False=@n23) {defaultValue=true, parameterName=USE_EyeSetup}
   )) {materialFunction=&MF_FX_EmissiveAdjustments}
  n34 = MaterialFunctionCall(Fade Length=Constant() {r=200}) {materialFunction=&CameraDepthFade}
  Multiply(A=@n34.Result) {constB=-1}
  n35 = TextureSample(UVs=Panner(Coordinate=TextureCoordinate() {vTiling=0.5714}, Time=@n9) {speedX=-0.1, speedY=1}) {texture=&T_Tiling_Noise_01}
  n36 = MaterialFunctionCall(Float2=TextureCoordinate()) {materialFunction=&BreakOutFloat2Components}
  n37 = TextureSample(UVs=MaterialFunctionCall(X=@n36.G, Y=Multiply(A=@n36.R) {constB=-1}) {materialFunction=&MakeFloat2}.Result) {texture=&T_Rust_Remix_T1+T2_v1_Gradient_Armor.T_Rust_Remix_T1+T2_v1_Gradient_Armor}
  n38 = Time()
  n39 = Power(
   Base=Multiply(A=@n37.RGB) {constB=4.5}
   Exp=MaterialFunctionCall(
    Input=Sine(None=@n38)
    Input Low=Constant() {r=-1}
    Input High=Constant() {r=1}
    Target Low=Constant() {r=1.2}
    Target High=Constant() {r=1.6}
   ) {materialFunction=&RemapValueRange}.Result
  ) {constExponent=1.4}
  n40 = Multiply(A=Multiply(A=Power(Base=@n35.RGB) {constExponent=0.5}) {constB=3}, B=@n39) {constB=2}
  n41 = Power(Base=Multiply(A=@n1), Exp=Constant() {r=2})
  n42 = Power(
   Base=@n1
   Exp=MaterialFunctionCall(Input=@n34.Result, Input Low=Constant() {r=0.3}, Input High=Constant() {r=1}, Target Low=Constant() {r=0.7}, Target High=Constant() {r=1.5}) {materialFunction=&RemapValueRange}.Result
  )
  Multiply(A=Clamp(None=Add(A=@n40, B=@n39)), B=Add(A=Clamp(None=@n41) {maxDefault=0.8}, B=Add(A=@n42, B=@n4)) {constB=0})
  Multiply(
   A=MaterialFunctionCall(Value_Morning=Constant() {r=0.1}, Value_Day=Constant() {r=1}, Value_Evening=Constant() {r=0.2}, Value_Night=Constant() {r=0.35}) {materialFunction=&Scalar_Time_of_Day}.Result
   B=@n32.RGB
  )
  Power(Base=@n1, Exp=Constant() {r=1.5})
  Power(Base=@n10)
  n43 = Subtract(
   A=Subtract(A=WorldPosition().XYZ, B=ActorPositionWS())
   B=Multiply(A=ScalarParameter() {defaultValue=50, parameterName=Curl_Speed}, B=Time())
  )
  n44 = ScalarParameter() {defaultValue=58.7749, parameterName=Curl_Scale}
  n45 = Multiply(
   A=ScalarParameter() {defaultValue=2, parameterName=Curl_Intensity}
   B=VectorNoise(World Position=Divide(A=@n43, B=@n44)) {noiseFunction=VNF_VectorALU}
  )
  n46 = Constant() {r=1}
  StaticSwitchParameter(True=Multiply(A=VertexColor().R, B=ShadingPathSwitch(Default=@n45, Mobile=@n46) {desc=Migrated from FeatureLevelSwitch}), False=Constant()) {parameterName=USE_WPO}
 ]
}
